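# Additive unlit glow material for the Sparks Greenroom island spotlight, with audio-reactive brightness.
# An orange Color parameter dims toward half as the saturated BeatLifetime value from the audio parameter collection rises, then feeds MF_FX_EmissiveAdjustments.
# A procedural mask combines a squared V ramp, a sine falloff across U and a smoothstepped inverse-V edge fade.
# Slowly panning blurred noise modulates that mask to break up the glow.
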Material /SparksEffects/Effects/Environment/Greenroom/IslandSpotlight/M_Sparks_Greenroom_Island_Glow
graph {
 properties {blendMode=BLEND_Additive, pixelDepthOffsetMode=PDOM_Legacy, shadingModel=MSM_Unlit, usedShadingModels=Unlit}
 nodes [
  n1 = Saturate(None=NamedRerouteUsage())
  MaterialFunctionCall(Emissive=Multiply(A=LinearInterpolate(Alpha=@n1) {constA=1, constB=0.5}, B=VectorParameter() {parameterName=Color, defaultValue=<rgba 1 0.1129 0 1>}.RGB)) {materialFunction=&MF_FX_EmissiveAdjustments}
  n2 = Power(Base=ComponentMask(None=TextureCoordinate()) {g=true}, Exp=Constant() {r=2})
  n3 = TextureCoordinate()
  n4 = Sine(None=Multiply(A=ComponentMask(None=@n3) {r=true}) {constB=0.5})
  n5 = Constant()
  n6 = Constant() {r=0.25}
  n7 = OneMinus(None=MaterialFunctionCall() {materialFunction=&LinearGradient}.VGradient)
  n8 = TextureCoordinate() {vTiling=0.25}
  Multiply(
   A=Multiply(A=Multiply(A=@n2, B=@n4), B=SmoothStep(Min=@n5, Max=@n6, Value=@n7))
   B=TextureSample(UVs=Panner(Coordinate=@n8) {speedX=0.025, speedY=0.1}) {samplerType=SAMPLERTYPE_LinearGrayscale, texture=/Game/Global/Textures/Default/Noise/LowResBlurredNoise_NoMips}.R
  )
  NamedRerouteDeclaration(None=CollectionParameter() {collection=/FMAudioReactiveData/AudioAnalysis/MPC_Sparks_Audio_Parameters, parameterName=BeatLifetime}) {name=BeatLifetime, nodeColor=<rgba 0 1 0.1797 1>}
 ]
}
