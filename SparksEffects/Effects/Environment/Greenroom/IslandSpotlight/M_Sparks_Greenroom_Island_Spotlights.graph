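# Additive unlit material for the island spotlight beams in the Sparks Greenroom environment.
# World position offset stretches vertices away from a light source point offset from the pivot, widens them with SplineThicken under a sphere mask, and slowly rotates them about a local axis.
# UV gradients and panning noise shape the beam.
# Color switches between squared vertex-color rainbow and a parameter, with brightness modulated by audio-reactive BeatLifetime.
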Material /SparksEffects/Effects/Environment/Greenroom/IslandSpotlight/M_Sparks_Greenroom_Island_Spotlights
graph {
 properties {blendMode=BLEND_Additive, pixelDepthOffsetMode=PDOM_Legacy, shadingModel=MSM_Unlit, twoSided=true, usedShadingModels=Unlit}
 nodes [
  n1 = ComponentMask(None=TextureCoordinate()) {r=true}
  n2 = Multiply(A=Sine(None=Multiply(A=@n1) {constB=0.5}), B=ComponentMask(None=TextureCoordinate()) {g=true})
  n3 = Constant() {r=0.1}
  n4 = VectorParameter() {parameterName=NoiseTile, defaultValue=<rgba 0.25 0.2 0 0>}
  n5 = Multiply(A=TextureCoordinate(), B=AppendVector(A=@n4.R, B=@n4.R))
  n6 = TextureCoordinate() {coordinateIndex=1}
  n7 = Add(A=Time(), B=VertexColor().R)
  n8 = ScalarParameter() {defaultValue=0.5, parameterName=PanSpeed}
  n9 = TextureSample(UVs=Panner(Coordinate=Add(A=@n5, B=@n6), Time=Multiply(A=@n7, B=@n8)) {speedX=0.1, speedY=1}) {texture=&T_Smooth_tile}
  Add(A=@n2, B=Multiply(A=@n2, B=LinearInterpolate(A=@n3, Alpha=@n9.G)))
  n10 = Constant3Vector() {constant=<rgba 0 0 -1 0>}
  n11 = Multiply(A=Time()) {constB=0.025}
  n12 = SphereMask(A=TextureCoordinate() {coordinateIndex=1}, B=Constant() {r=0.5}, Radius=ScalarParameter() {defaultValue=1, parameterName=SphereRadius}) {attenuationRadius=1, hardnessPercent=0}
  n13 = ScalarParameter() {defaultValue=25000, parameterName=Length}
  n14 = ScalarParameter() {defaultValue=30, parameterName=LightSourceOffset}
  n15 = Normalize(VectorInput=Transform(None=Constant3Vector() {constant=<rgba 0 0 1 0>}) {transformSourceType=TRANSFORMSOURCE_Local})
  n16 = Subtract(
   A=WorldPosition().XYZ
   B=Add(
    A=Multiply(A=@n14, B=@n15)
    B=MaterialFunctionCall() {materialFunction=&ObjectPivotPoint}.Object Pivot Location
   )
  )
  n17 = LinearInterpolate(A=Multiply(A=Multiply(A=@n12, B=@n13), B=Normalize(VectorInput=@n16)), Alpha=ComponentMask(None=TextureCoordinate()) {g=true}) {constB=0}
  n18 = WorldPosition()
  n19 = Multiply(A=ScalarParameter() {defaultValue=4000, parameterName=SplineWidth}, B=@n12)
  n20 = Multiply(A=@n19, B=ScalarParameter() {defaultValue=0.25, parameterName=SplineWidthTipMult})
  Add(
   A=RotateAboutAxis(
    NormalizedRotationAxis=Transform(None=@n10) {transformSourceType=TRANSFORMSOURCE_Local}
    RotationAngle=Frac(None=@n11)
    PivotPoint=ActorPositionWS()
    Position=Add(A=@n17, B=@n18.XYZ)
   )
   B=Add(
    A=MaterialFunctionCall(WidthBase=@n20, WidthTip=@n19) {materialFunction=&SplineThicken}.WorldPositionOffset
    B=@n17
   )
  )
  n21 = Saturate(None=NamedRerouteUsage())
  n22 = VertexColor()
  n23 = Multiply(A=@n22, B=@n22)
  n24 = VectorParameter() {parameterName=Color, defaultValue=<rgba 0 0 0 1>}
  MaterialFunctionCall(Emissive=Multiply(
    A=LinearInterpolate(Alpha=@n21) {constA=1, constB=0.5}
    B=StaticSwitchParameter(True=@n23, False=@n24.RGB) {defaultValue=true, parameterName=Use Rainbow}
   )) {materialFunction=&MF_FX_EmissiveAdjustments}
  NamedRerouteDeclaration(None=CollectionParameter() {collection=/FMAudioReactiveData/AudioAnalysis/MPC_Sparks_Audio_Parameters, parameterName=BeatLifetime}) {name=BeatLifetime, nodeColor=<rgba 0 1 0.1797 1>}
 ]
}
